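; SilkyBingo destructible Athena prop that bursts on death, applying water and an environmental Slurp effect nearby.
; Server death, gated on allowed death effects, sphere-overlaps a scale-adjusted radius, applies both effects to hit actors, then reports an extinguish quest stat if flagged.
; Death effects spawn Niagara and a randomly rotated, fading splash decal, and hide the mesh to avoid relevancy reappearance.
; ShouldDie always returns true.
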(event EventOnDeathServer
  (params (Damage :type Float%20%28single-precision%29) (DamageTags :type Gameplay%20Tag%20Container%20Structure%20%28by%20ref%29) (Momentum :type Vector :default "0, 0, 0") (HitInfo :type Hit%20Result%20Structure%20%28by%20ref%29) (InstigatedBy :type Controller%20Object%20Reference) (DamageCauser :type Actor%20Object%20Reference) (EffectContext :type Gameplay%20Effect%20Context%20Handle%20Structure))
  (if (Class|AthenaFunctionLibrary|ShouldAllowDeathEffects DamageTags)
    (CallFunction|DeactivateEmitter)
    (Variables|Default|SetWaterInstigator InstigatedBy)
    (bind (_outactors _returnvalue) (Collision|SphereOverlapActors (Transformation|GetActorLocation) (* (Variables|Default|GetExplosion%20Radius) (Math|Vector|VectorLength (Transformation|GetActorScale3D))) (Variables|Default|GetObject%20Types)))
    (if _returnvalue
      (for _array_element _outactors
        (Utilities|IsValid _array_element
          (:Is%20Valid
            (bind _output_get (Variables|Default|SetHitActor _array_element))
            ;[1 1 1] Apply water to hit actors.
            (CallFunction|ApplyWaterIfCurieEnabled (Variables|Default|GetWaterInstigator) _output_get)
            (bind _asfort_pawn (Utilities|Casting|CastToFortPawn (Variables|Default|GetHitActor)))
            (GameplayEffects|FortApplyGameplayEffectToTarget (Ability|GetFortAbilitySystemComponent self) "/Game/Athena/Items/Gameplay/SilkyBingo/GE_Athena_EnvSlurp_Grant_SilkyBingo.GE_Athena_EnvSlurp_Grant_SilkyBingo_C" (Ability|GetAbilitySystemComponent _asfort_pawn) self self 0 0.0 (select (Utilities|IsValid (Ability|GetAbilitySystemComponent (Variables|Default|GetWaterInstigator)) :!class /Script/Engine.KismetSystemLibrary) (Class|AthenaFunctionLibrary|MakeSimpleFortEffectContext (Variables|Default|GetWaterInstigator)) 0)))))
      (CallFunction|QuestHandling))))

(event EventOnDeathPlayEffects
  (params (Damage :type Float%20%28single-precision%29) (DamageTags :type Gameplay%20Tag%20Container%20Structure%20%28by%20ref%29) (Momentum :type Vector :default "0, 0, 0") (HitInfo :type Hit%20Result%20Structure%20%28by%20ref%29) (InstigatedBy :type Fort%20Pawn%20Object%20Reference) (DamageCauser :type Actor%20Object%20Reference) (EffectContext :type Gameplay%20Effect%20Context%20Handle%20Structure))
  (|CallParentFunction Damage DamageTags Momentum HitInfo InstigatedBy DamageCauser EffectContext)
  (Niagara|SpawnSystematLocation :SystemTemplate "/Game/Athena/Items/Gameplay/SilkyBingo/FX/NS_SilkyBingo_Explode_Update.NS_SilkyBingo_Explode_Update" :Location (Transformation|GetActorLocation) :Scale "1.000000,1.000000,1.000000" :bAutoDestroy true :bAutoActivate true :PoolingMethod "AutoRelease" :bPreCullCheck true)
  ;[0.15 0.1487 0 0.5] FORT-1043968 - Prevents object from reapiring due relevancy issues.
  (Rendering|SetVisibility (Variables|Debug|GetStaticMeshComponent) false false)
  (bind _returnvalue (Rendering|Decal|SpawnDecalatLocation "/Game/Athena/Items/Gameplay/SilkyBingo/FX/MI_Decal_SilkyBingo_Splash.MI_Decal_SilkyBingo_Splash" (Math|Vector|MakeVector (* (.x (Variables|Default|GetDeath%20Decal%20Size)) (.x (Transformation|GetActorScale3D))) (* (.y (Variables|Default|GetDeath%20Decal%20Size)) (.y (Transformation|GetActorScale3D))) (* (.z (Variables|Default|GetDeath%20Decal%20Size)) (.z (Transformation|GetActorScale3D)))) (Transformation|GetActorLocation) (Math|Rotator|MakeRotator (Math|Random|RandomFloatinRangefromStream (Variables|Default|GetRandom%20Stream) 0.0 359.0) -9e+01)))
  (Rendering|Components|Decal|SetFadeIn _returnvalue 0.0 0.25)
  (Rendering|Components|Decal|SetFadeOut _returnvalue 0.75 1.75 true))

(fn DistanceCheckforCameraLensEffect
  (params (Instigator :type Actor%20Object%20Reference))
  (bind (_returnvalue_x _returnvalue_y _returnvalue_z) (Math|Vector|BreakVector (Transformation|GetActorLocation)))
  (bind (_returnvalue_x_2 _returnvalue_y_2 _returnvalue_z_2) (Math|Vector|BreakVector (Transformation|GetActorLocation Instigator)))
  (return (Math|Boolean|ORBoolean (or (> _returnvalue_x_2 (+ _returnvalue_x (Variables|Default|GetDistance%20for%20Camera%20Lens))) (< _returnvalue_x_2 (- _returnvalue_x (Variables|Default|GetDistance%20for%20Camera%20Lens)))) (or (> (+ _returnvalue_y (Variables|Default|GetDistance%20for%20Camera%20Lens)) _returnvalue_y_2) (< (- _returnvalue_y (Variables|Default|GetDistance%20for%20Camera%20Lens)) _returnvalue_y_2)) (or (> _returnvalue_z_2 (+ _returnvalue_z (Variables|Default|GetDistance%20for%20Camera%20Lens))) (< _returnvalue_z_2 (- _returnvalue_z (Variables|Default|GetDistance%20for%20Camera%20Lens)))))))

(fn QuestHandling
  (bind _waterinstigator (Variables|Default|GetWaterInstigator))
  (bind _returnvalue (Fortnite|GetControllerfromActor _waterinstigator))
  (if (Variables|Default|GetSomethingWasPutOut)
    (Utilities|IsValid _returnvalue
      (:Is%20Valid
        (bind _asfort_player_controller (Utilities|Casting|CastToFortPlayerController _returnvalue))
        (Quest|SendComplexCustomStatEvent (Quests|GetQuestManager _asfort_player_controller "Invalid") 0 "(GameplayTags=((TagName=\"Athena.Quest.Extinguish\")))" "(GameplayTags=)" (Variables|Default|GetCount))
        (Variables|Default|SetSomethingWasPutOut false)))))

(fn ShouldDie
  (params (Damage :type Float%20%28single-precision%29) (EventInstigator :type Controller%20Object%20Reference) (DamageCauser :type Actor%20Object%20Reference) (bIsTest :type Boolean) (EffectContext :type Gameplay%20Effect%20Context%20Handle%20Structure%20%28by%20ref%29))
  (return true))

(fn ConstructionScript
  (|CallParentFunction))
